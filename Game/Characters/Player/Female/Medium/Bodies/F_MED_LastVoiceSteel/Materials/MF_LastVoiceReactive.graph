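# Material function that adds a reactive color and glow effect to the LastVoiceSteel female character body's material attributes, overriding BaseColor and Emissive.
# A reactivity parameter, or debug time loop, appears to move and resize a sphere mask along pre-skinned Z.
# Panned, warped noise projected on two pre-skinned planes breaks up that gradient.
# Curve atlas sampling supplies color, with Fresnel and an alpha input (body FX texture preview) shaping the effect.
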MaterialFunction /Game/Characters/Player/Female/Medium/Bodies/F_MED_LastVoiceSteel/Materials/MF_LastVoiceReactive
graph {nodes [
  FunctionOutput(None=SetMaterialAttributes!(MaterialAttributes=NamedRerouteUsage(), BaseColor=NamedRerouteUsage(), EmissiveColor=NamedRerouteUsage()) {attributeSetTypes [#1, #2]})
  n1 = GetMaterialAttributes(None=FunctionInput() {bUsePreviewValueAsDefault=true, inputName=Material, inputType=FunctionInput_MaterialAttributes}) {attributeGetTypes [#3, #1, #2]}
  NamedRerouteDeclaration(None=@n1.EmissiveColor) {name=Emissive_IN, nodeColor=<rgba 0 1 0.3438 1>}
  NamedRerouteDeclaration(None=VertexInterpolator(VS=PreSkinnedPosition()).PS) {name=PreskinLocation, nodeColor=<rgba 0.9453 0 1 1>}
  n2 = NamedRerouteUsage()
  n3 = LinearInterpolate(A=Constant3Vector() {constant=<rgba 0 0 -120 0>}, B=Constant3Vector() {constant=<rgba 0 0 300 0>}, Alpha=NamedRerouteUsage())
  n4 = NamedRerouteUsage()
  n5 = Constant()
  n6 = Power(
   Base=Subtract(A=SphereMask(A=@n2, B=@n3, Radius=@n4, Hardness=@n5) {attenuationRadius=100, hardnessPercent=0}, B=NamedRerouteUsage())
   Exp=ScalarParameter() {defaultValue=0.5, group=LastVoiceSteel, parameterName=LastVoiceSteel_SphereMaskPower}
  )
  n7 = Multiply(A=Saturate(None=@n6), B=NamedRerouteUsage())
  n8 = ScalarParameter() {group=LastVoiceSteel, parameterName=LastVoiceSteel_SpherePatternBlend}
  NamedRerouteDeclaration(None=Multiply(
    A=LinearInterpolate(A=@n6, B=@n7, Alpha=@n8)
    B=ScalarParameter() {defaultValue=1, group=LastVoiceSteel, parameterName=LastVoiceSteel_ReactivityGradientIntensity}
   )) {name=ReactivityGradient, nodeColor=<rgba 0 1 0.9297 1>}
  n9 = Multiply(A=NamedRerouteUsage(), B=ScalarParameter() {defaultValue=1, group=LastVoiceSteel, parameterName=LastVoiceSteel_WarpScale})
  n10 = Multiply(
   A=Fresnel(ExponentIn=ScalarParameter() {defaultValue=1, group=LastVoiceSteel, parameterName=LastVoiceSteel_CurveFresnel_Exp}) {baseReflectFraction=0}
   B=ScalarParameter() {defaultValue=0.8, group=LastVoiceSteel, parameterName=LastVoiceSteel_CurveFresnel_Intensity}
  )
  n11 = CurveAtlasRowParameter(CurveTime=Saturate(None=Add(A=@n9, B=@n10))) {
   atlas=/Game/Characters/Player/Female/Medium/Bodies/F_MED_LastVoiceSteel/Textures/CA_LastVoiceSteel
   curve=/Game/Characters/Player/Female/Medium/Bodies/F_MED_LastVoiceSteel/Textures/C_LastVoiceSteel
   group=LastVoiceSteel
   parameterName=LastVoiceSteel_Curve
  }
  NamedRerouteDeclaration(None=@n11) {name=SampledCurve, nodeColor=<rgba 1 0.1121 0.984 1>}
  n12 = Divide(A=NamedRerouteUsage(), B=Constant() {r=128}) {constB=128}
  n13 = ComponentMask(None=@n12) {b=true, r=true}
  n14 = ComponentMask(None=VectorParameter() {group=LastVoiceSteel, parameterName=LastVoiceSteel_NoisePanSpeed, defaultValue=<rgba 0.05 -0.2 0 1>}.RGB) {g=true, r=true}
  n15 = Panner(Coordinate=@n13, Speed=@n14)
  n16 = Constant2Vector() {r=0.5}
  n17 = TextureSample(
   UVs=Panner(
    Coordinate=Add(A=@n13, B=@n16)
    Speed=VectorParameter() {group=LastVoiceSteel, parameterName=LastVoiceSteel_NoiseWarpSpeed, defaultValue=<rgba 0 -0.05 0 1>}.RGB
   ) {speedX=0.2, speedY=0.5}
   Tex=NamedRerouteUsage()
  ) {samplerSource=SSM_Wrap_WorldGroupSettings, samplerType=SAMPLERTYPE_Grayscale, texture=&LowResBlurredNoise}
  n18 = Multiply(
   A=ConstantBiasScale(None=Saturate(None=@n17.G)) {bias=-0.5, scale=2}
   B=ScalarParameter() {defaultValue=0.1, group=LastVoiceSteel, parameterName=LastVoiceSteel_NoiseWarpIntensity}
  )
  n19 = ComponentMask(None=NamedRerouteUsage()) {g=true, r=true}
  n20 = ScalarParameter() {defaultValue=0.03, group=LastVoiceSteel, parameterName=LastVoiceSteel_NormalWarpIntensity}
  n21 = TextureSample(UVs=Add(A=Add(A=@n15, B=@n18), B=Multiply(A=@n19, B=@n20)) {desc=UV Warp}, Tex=NamedRerouteUsage()) {samplerSource=SSM_Wrap_WorldGroupSettings, samplerType=SAMPLERTYPE_Grayscale, texture=&LowResBlurredNoise}
  n22 = TextureSample(UVs=Panner(Coordinate=ComponentMask(None=@n12) {b=true, g=true}, Speed=@n14), Tex=NamedRerouteUsage()) {samplerSource=SSM_Wrap_WorldGroupSettings, samplerType=SAMPLERTYPE_Grayscale, texture=&LowResBlurredNoise}
  n23 = MaterialFunctionCall(Float3=LinearInterpolate(A=@n21.RGB, B=@n22.RGB, Alpha=NamedRerouteUsage())) {materialFunction=&BreakOutFloat3Components}
  NamedRerouteDeclaration(None=@n23.R) {name=PerlinNoise, nodeColor=<rgba 1 0 0.3047 1>}
  NamedRerouteDeclaration(None=@n23.G) {name=Pattern, nodeColor=<rgba 0.4297 0 1 1>}
  n24 = TextureSample() {
   samplerType=SAMPLERTYPE_Masks
   texture=/Game/Characters/Player/Female/Medium/Bodies/F_MED_LastVoiceSteel/Textures/T_F_MED_LastVoiceSteel_Body_FX
  }
  NamedRerouteDeclaration(None=FunctionInput(Preview=@n24.R) {inputName=Alpha, inputType=FunctionInput_Scalar}) {name=Reactivity_Alpha, nodeColor=<rgba 0.9922 1 0 1>}
  n25 = TextureObjectParameter() {
   group=LastVoiceSteel
   parameterName=LastVoiceSteel_ReactivityNoiseTex
   samplerType=SAMPLERTYPE_Masks
   texture=/Game/Characters/Player/Female/Medium/Bodies/F_MED_LastVoiceSteel/Textures/T_F_MED_LastVoiceSteel_TA
  }
  NamedRerouteDeclaration(None=@n25) {name=ReactivityTex, nodeColor=<rgba 0 1 0.9297 1>}
  n26 = VertexInterpolator(VS=PreSkinnedNormal())
  n27 = ComponentMask(None=SmoothStep(Value=Normalize(VectorInput=@n26.PS))) {b=true}
  NamedRerouteDeclaration(None=Multiply(A=Saturate(None=@n27), B=Constant() {r=1})) {name=PlanarBlend, nodeColor=<rgba 0.7344 1 0 1>}
  NamedRerouteDeclaration(None=@n1.BaseColor) {name=BaseColor_IN, nodeColor=<rgba 0 1 0.2031 1>}
  NamedRerouteDeclaration(None=LinearInterpolate(A=NamedRerouteUsage(), B=NamedRerouteUsage(), Alpha=NamedRerouteUsage())) {name=BaseColor_Out, nodeColor=<rgba 0 1 0.1328 1>}
  n28 = Time()
  n29 = ScalarParameter() {defaultValue=-0.2, group=LastVoiceSteel, parameterName=LastVoiceSteel_DebugReactiveSpeed}
  n30 = OneMinus(None=Frac(None=Multiply(A=@n28, B=@n29)))
  n31 = LinearInterpolate(
   A=ScalarParameter() {group=LastVoiceSteel, parameterName=LastVoiceSteel_Reactivity}
   B=Constant() {r=0.5}
   Alpha=Clamp(None=ScalarParameter() {group=LastVoiceSteel, parameterName=LastVoiceSteel_ColorMode})
  )
  NamedRerouteDeclaration(None=LinearInterpolate(
    B=StaticSwitchParameter(True=@n30, False=@n31) {group=LastVoiceSteel, parameterName=LastVoiceSteel_UseDebug}
    Alpha=ScalarParameter() {defaultValue=1, group=LastVoiceSteel, parameterName=LastVoiceSteel_ReactivityToggle}
   )) {name=time, nodeColor=<rgba 1 0 0.9375 1>}
  NamedRerouteDeclaration(None=@n1.MaterialAttributes) {name=MA_IN, nodeColor=<rgba 0.2891 0 1 1>}
  n32 = Multiply(
   A=Fresnel(
    ExponentIn=ScalarParameter() {defaultValue=1.2, group=LastVoiceSteel, parameterName=LastVoiceSteel_FresnelExp}
    BaseReflectFractionIn=ScalarParameter() {defaultValue=0.05, group=LastVoiceSteel, parameterName=LastVoiceSteel_FresnelBase}
   )
   B=ScalarParameter() {defaultValue=1.8, group=LastVoiceSteel, parameterName=LastVoiceSteel_FresnelStrength}
  )
  n33 = Saturate(None=NamedRerouteUsage())
  NamedRerouteDeclaration(None=LinearInterpolate(B=LinearInterpolate(B=@n32, Alpha=@n33), Alpha=NamedRerouteUsage())) {name=ReactivityMask, nodeColor=<rgba 1 0.9141 0 1>}
  n34 = NamedRerouteUsage()
  n35 = ScalarParameter() {defaultValue=2, group=LastVoiceSteel, parameterName=LastVoiceSteel_ReactiveEmissiveStrength}
  NamedRerouteDeclaration(None=LinearInterpolate(A=NamedRerouteUsage(), B=Multiply(A=@n34, B=@n35), Alpha=NamedRerouteUsage())) {name=Emissive_OUT, nodeColor=<rgba 0 1 0.6484 1>}
  NamedRerouteDeclaration(None=@n1.Normal) {name=Normal_IN, nodeColor=<rgba 0.6598 0.4271 1 1>}
  n36 = NamedRerouteUsage()
  n37 = ScalarParameter() {defaultValue=0.5, group=LastVoiceSteel, parameterName=LastVoiceSteel_ReactivityMax}
  n38 = Divide(A=Abs(None=Subtract(A=@n36, B=@n37) {constB=0.5}), B=@n37)
  NamedRerouteDeclaration(None=LinearInterpolate(A=Constant() {r=150}, B=Constant() {r=300}, Alpha=OneMinus(None=@n38))) {name=SphereMaskRadius, nodeColor=<rgba 0.7344 1 0 1>}
 ]}
