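# Master skeletal-mesh material for the LastVoiceSteel female player body, built on material attributes between shared character PreFX and PostFX functions.
# A skin FX mask texture drives switchable thin film and slurp layers, followed by cloth-sim WPO and emissive fresnel.
# Static switches add eye UVs with eye post-processing and the LastVoice reactive function.
# A final stage scales emissive and blends base color toward a scaled copy before output.
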Material /Game/Characters/Player/Female/Medium/Bodies/F_MED_LastVoiceSteel/Materials/M_LastVoiceSteel_MASTER
graph {
 properties {
  bUseMaterialAttributes=true
  bUsedWithMeshDeformer=true
  bUsedWithMorphTargets=true
  bUsedWithSkeletalMesh=true
  pixelDepthOffsetMode=PDOM_Legacy
  usedShadingModels=Default Lit
 }
 nodes [
  n1 = MaterialFunctionCall() {materialFunction=/Game/Characters/Shared/FNEye/Materials/MF_EyeUVs}
  n2 = MaterialFunctionCall(TexCoords=StaticSwitchParameter(True=@n1.EyeUVs, False=TextureCoordinate()) {group=LastVoiceSteel, parameterName=UseEyes}) {materialFunction=&BaseFN_CharacterFunction_PreFX}
  MaterialFunctionCall(
   AO=NamedRerouteDeclaration(None=@n2.AO) {name=AO, nodeColor=<rgba 0 1 0.4844 1>}
   Distance Blend=@n2.Distance Blend
   Fresnel=@n2.Distance Blend
   Gmap=@n2.Distance Blend
   Material=NamedRerouteUsage()
  ) {materialFunction=&BaseFN_CharacterFunction_PostFX}
  NamedRerouteDeclaration(None=@n2.Pre skinned Local Position) {name=PreskinnedLocalPos, nodeColor=<rgba 0 1 0.8359 1>}
  n3 = NamedRerouteUsage()
  n4 = NamedRerouteDeclaration(None=TextureSample(Tex=TextureObjectParameter() {group=Skin FX, parameterName=SkinFX_Mask}).RGBA) {name=SkinFX_Mask, nodeColor=<rgba 1 0 0.0703 1>}
  n5 = NamedRerouteUsage()
  n6 = StaticSwitchParameter(
   True=MaterialFunctionCall!(Alpha=ChannelMaskParameter(None=@n4) {parameterName=ThinFilmChannel}.R, Material Attributes=@n5) {materialFunction=/Game/Weapons/WeaponSkins/MaterialFunction/ThinFilmFunction}.Result
   False=@n5
  ) {group=Skin FX, parameterName=Use Thin Film}
  n7 = MaterialFunctionCall!(
   Alpha=NamedRerouteDeclaration(None=StaticComponentMaskParameter(None=@n4) {defaultR=true, group=Skin FX, parameterName=Slurp Channel}) {name=SlurpMask, nodeColor=<rgba 0.2656 0 1 1>}
   AO=NamedRerouteUsage()
   BubblesMask=StaticComponentMaskParameter(None=@n4) {defaultG=true, group=Skin FX, parameterName=Slurp Bubbles Channel}
   SlurpNormalDistortionMask=StaticComponentMaskParameter(None=@n4) {defaultB=true, group=Skin FX, parameterName=Slurp Normal Distortion Channel}
   Pre Skinned Position=NamedRerouteUsage()
   Material=@n6
  ) {materialFunction=/Game/Weapons/WeaponSkins/MaterialFunction/MF_Slurp}
  n8 = MaterialFunctionCall!(In=MaterialFunctionCall!(Material=StaticSwitchParameter(True=@n7.Result, False=@n6) {group=Skin FX, parameterName=UseSlurp}, SkinFX_Mask=NamedRerouteUsage()) {materialFunction=/Game/Global/MaterialFunctions/MF_ClothSimWPO_v2}.Result) {materialFunction=/Game/Weapons/WeaponSkins/MaterialFunction/MF_EmissiveFresnel}
  n9 = StaticSwitchParameter(
   True=MaterialFunctionCall!(MA_IN=@n8.Result, EyeUVs=@n1.EyeUVs, EyeNormal=@n1.EyeUVs, UVs=@n1.EyeUVs, Debug=@n1.EyeUVs, DebugCornea=@n1.EyeUVs) {materialFunction=/Game/Characters/Shared/FNEye/Materials/MF_Eyes_Post}.Result
   False=@n8.Result
  ) {group=LastVoiceSteel, parameterName=UseEyes}
  n10 = GetMaterialAttributes(None=StaticSwitchParameter(
    True=MaterialFunctionCall!(Alpha=@n3, Material=@n9) {materialFunction=&/MF_LastVoiceReactive}.Result
    False=@n9
   ) {group=LastVoiceSteel, parameterName=UseReactivity}) {attributeGetTypes [#1, #2]}
  n11 = ScalarParameter() {defaultValue=0.4, group=LastVoiceSteel, parameterName=LastVoiceSteel_EmissiveBoost}
  n12 = NamedRerouteUsage()
  n13 = Multiply(A=@n12, B=ScalarParameter() {defaultValue=0.25, group=LastVoiceSteel, parameterName=LastVoiceSteel_BaseColorBoost})
  n14 = NamedRerouteUsage()
  NamedRerouteDeclaration(None=SetMaterialAttributes!(
    MaterialAttributes=@n10.MaterialAttributes
    Emissive Color=Multiply(A=@n10.Emissive Color, B=@n11)
    Base Color=LinearInterpolate(A=@n12, B=@n13, Alpha=@n14)
   ) {attributeSetTypes [#1, #2]}) {name=MA_OUT, nodeColor=<rgba 1 0.9141 0 1>}
  NamedRerouteDeclaration(None=@n2.Result) {name=MA_IN, nodeColor=<rgba 1 0.5859 0 1>}
  NamedRerouteDeclaration(None=GetMaterialAttributes(None=NamedRerouteUsage()) {attributeGetTypes [#2]}.Base Color) {name=BaseColor_IN, nodeColor=<rgba 0 1 0.2734 1>}
 ]
}
